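# Looping indoor ambience cue for the PirateCave POI, built on NitroWasteland pirate cave recordings with Athena ambient attenuation and the indoor sound class.
# A quality-level switch plays a looping quad wave for its first two levels and a stereo wave for the last two.
# Two nested, endlessly looping envelopers add slow variation: the inner cycles pitch, while the outer holds pitch lowered and gently dips volume over a longer period.
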SoundCue /Game/Sounds/Helios/POI/PirateCave/Ambience_PirateCave_Cave_Loop_Cue
root=Enveloper_0
nodes [
 {
  id=Enveloper_0
  children [Enveloper_3]
  properties {
   bLoop=true
   bLoopIndefinitely=true
   loopEnd=100
   pitchCurve {editorCurveData {keys [<at 0 0.8>]}}
   volumeCurve {editorCurveData {keys [<at 0 1 RCIM_Cubic>,<at 50 0.9 RCIM_Cubic>,<at 100 1 RCIM_Cubic>]}}
  }
 }
 {
  id=Enveloper_3
  children [QualityLevel_0]
  properties {
   bLoop=true
   bLoopIndefinitely=true
   loopEnd=30
   pitchCurve {editorCurveData {keys [{interpMode=RCIM_Cubic, time=0}, <at 15.9468 0.9181 RCIM_Cubic>, {interpMode=RCIM_Cubic, time=30}]}}
   volumeCurve {editorCurveData {keys [<at 0 1>]}}
  }
 }
 {id=QualityLevel_0, children [WavePlayer_12, WavePlayer_12, WavePlayer_6, WavePlayer_6]}
 {
  id=WavePlayer_12
  properties {bLooping=true, soundWaveAssetPtr=&/Ambience_NitroWasteland_Pirate_Cave_01_Quad}
 }
 {
  id=WavePlayer_6
  properties {bLooping=true, soundWaveAssetPtr=&/Ambience_NitroWasteland_Pirate_Cave_01_Stereo}
 }
]
properties {
 attenuationSettings=&Fort_Ambient_Athena_Att
 maxDistance=2097152
 soundClassObject=&Fort_Ambient_Indoors_Athena
 volumeMultiplier=0.301
}
